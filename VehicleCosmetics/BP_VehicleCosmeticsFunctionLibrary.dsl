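; Blueprint function library of shared helpers for vehicle cosmetic actors.
; It applies additional color, float and int parameters to Niagara components, using the Painted data table color when no color parameters exist.
; It sets IsPainted and ColorVariantID material scalars from a color index.
; It shifts skeletal meshes by the body-to-right-turbo socket X/Z difference plus an offset, and unpacks wheel FX info from the vehicle interface.
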(fn CosmeticActorColorReady
  (params (Color :type Name) (NiagaraColorVariable :type Name) (AdditionalInfo :type Vehicle%20Cosmetics%20Additional%20Parameter%20Infos%20Structure) (Niagara%20Component :type Niagara%20Particle%20System%20Component%20Object%20Reference) (Interface :type Vehicle%20Cosmetics%20Interface%20Interface) (chassisBoostColor :type Linear%20Color%20Structure))
  (bind (_floatparameters _intparameters _colorparameters) (Utilities|Struct|BreakVehicleCosmeticsAdditionalParameterInfos AdditionalInfo))
  (Utilities|IsValid Niagara%20Component
    (:Is%20Valid
      (Utilities|FlowControl|Sequence
        (:then_0
          (if (> (Utilities|Array|Length _colorparameters) 0)
            (for _array_element _colorparameters
              (Niagara|SetNiagaraVariable(LinearColor) Niagara%20Component (Utilities|Struct|BreakVehicleCosmeticsCustomizableObjectColorInfo _array_element) (.ParameterValue (Utilities|Struct|BreakVehicleCosmeticsCustomizableObjectColorInfo _array_element))))
            (else
              (bind _returnvalue (Utilities|GetDataTableRow "/VehicleCosmetics/DataTables/DT_VehicleCosmetics_Painted.DT_VehicleCosmetics_Painted" Color))
              (Niagara|SetNiagaraVariable(LinearColor) Niagara%20Component NiagaraColorVariable (Utilities|Struct|BreakStructVehicleCosmeticsPainted _returnvalue)))))
        (:then_1
          (for _array_element_2 _floatparameters
            (Niagara|SetNiagaraVariable(Float) Niagara%20Component (Utilities|Struct|BreakVehicleCosmeticsCustomizableObjectFloatInfo _array_element_2) (.ParameterValue (Utilities|Struct|BreakVehicleCosmeticsCustomizableObjectFloatInfo _array_element_2)))))
        (:then_2
          (for _array_element_3 _intparameters
            (Niagara|SetNiagaraVariable(Int32) Niagara%20Component (Utilities|Struct|BreakVehicleCosmeticsCustomizableObjectIntInfo _array_element_3) (Utilities|String|StringToInteger (Utilities|String|ToString(Name) (.ParameterValue (Utilities|Struct|BreakVehicleCosmeticsCustomizableObjectIntInfo _array_element_3)))))))))))

(fn CosmeticActorSetColorVariantIDToMaterials
  (params (Color :type Name) (Parent :type Object%20Reference) (Skeletal%20Mesh :type Skeletal%20Mesh%20Component%20Object%20Reference))
  (bind _skeletal_mesh (Variables|GetSkeletal%20Mesh))
  (bind _color_index (Class|BPVehicleCosmeticsFunctions|BPGetNSIndexbyColor Color Parent))
  (Rendering|Material|SetScalarParameterValueonMaterials _skeletal_mesh "IsPainted" (Math|Conversions|ToFloat(Integer) (Math|Integer|Clamp(Integer) _color_index 0 1)))
  (Rendering|Material|SetScalarParameterValueonMaterials _skeletal_mesh "ColorVariantID" (Math|Conversions|ToFloat(Integer) _color_index)))

(fn CosmeticActorSetSkeletalAnimLocation
  (params (Target :type Vehicle%20Cosmetics%20Actor%20Object%20Reference) (Offset :type Vector) (SkeletalMesh :type Skeletal%20Mesh%20Component%20Object%20Reference))
  (bind _target (Variables|GetTarget))
  (bind _returnvalue (Vehicle|Cosmetics|GetBodyMeshComponent _target))
  (bind _returnvalue_location (Math|Transform|BreakTransform (Transformation|GetSocketTransform _returnvalue "Socket_Body" "RTS_ParentBoneSpace")))
  (bind (_returnvalue_location_x _returnvalue_location_y _returnvalue_location_z) (Math|Vector|BreakVector _returnvalue_location))
  (bind _returnvalue_location_2 (Math|Transform|BreakTransform (Transformation|GetSocketTransform _returnvalue "Socket_Turbo_A_R" "RTS_ParentBoneSpace")))
  (bind (_returnvalue_location_x_2 _returnvalue_location_y_2 _returnvalue_location_z_2) (Math|Vector|BreakVector _returnvalue_location_2))
  (Utilities|IsValid _returnvalue
    (:Is%20Valid
      (Transformation|AddRelativeLocation (Variables|GetSkeletalMesh) (+ (Math|Vector|MakeVector (- _returnvalue_location_x_2 _returnvalue_location_x) 0.0 (- _returnvalue_location_z_2 _returnvalue_location_z)) (Variables|GetOffset)) false false)
      (return))
    (:Is%20Not%20Valid
      (Utilities|Debugging|RaiseScriptError "Get body component failed"))))

(fn GetWheelFXInfo
  (params (Vehicle%20Interface :type Fort%20Vehicle%20Interface%20Interface) (Wheel%20Location :type EVehicleWheelLocations%20Enum :default "FrontRight"))
  (return (Utilities|Struct|BreakVehicleWheelFXInfo (Vehicle|GetWheelFXInfo Vehicle%20Interface Wheel%20Location))))
