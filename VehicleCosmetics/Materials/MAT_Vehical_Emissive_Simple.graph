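# Unlit, masked emissive material for vehicle cosmetics on skeletal meshes, with a Kusanagi T3 hologram texture as the default mask.
# The mask's red and green channels tint CustomColor and TeamColor respectively, and the two are summed.
# One scalar lerps from a flat Emissive_Color to that masked tint.
# A second scalar lerps from solid alpha to the texture's alpha channel, so color and cutout switch independently.
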Material /VehicleCosmetics/Materials/MAT_Vehical_Emissive_Simple
graph {
 properties {bAutomaticallySetUsageInEditor=false, bUsedWithSkeletalMesh=true, blendMode=BLEND_Masked, shadingModel=MSM_Unlit, usedShadingModels=Unlit}
 nodes [
  n1 = TextureSampleParameter2D() {
   group=00_Base
   parameterName=Masks
   samplerType=SAMPLERTYPE_LinearColor
   sortPriority=4
   texture=/VehicleCosmetics/Bodies/Kusanagi/Kusanagi_T3/Textures/T_KusanagiT3_Holograms_RGB
   channelNames {a=Opacity, b=None, g=Team Color, r=Custom Color}
  }
  n2 = VectorParameter() {group=00_Base, parameterName=TeamColor, sortPriority=2, defaultValue=<rgba 0 0.2854 1 1>}
  n3 = VectorParameter() {group=00_Base, parameterName=CustomColor, sortPriority=1, defaultValue=<rgba 1 0 0.949 1>}
  LinearInterpolate(
   A=VectorParameter() {group=00_Base, parameterName=Emissive_Color, sortPriority=0}.RGB
   B=Add(A=Multiply(A=@n1.Team Color, B=@n2.RGB), B=Multiply(A=@n1.Custom Color, B=@n3.RGB))
   Alpha=ScalarParameter() {group=00_Base, parameterName=Solid Color 0 / Masked 1, sortPriority=3}
  )
  LinearInterpolate(B=@n1.Opacity, Alpha=ScalarParameter() {group=00_Base, parameterName=Solid Alpha 0 / Masked Alpha 1, sortPriority=3}) {constA=1}
 ]
}
