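# Clear Coat vehicle body material that, judging by its name and layout, serves as a parameter container for Mutable-based cosmetics rather than a working shader.
# It declares color, LOD, trim sheet transform and mirroring, logo normal, and base texture parameters, mostly with placeholder textures.
# Multiply nodes chain these parameters together arithmetically without meaningful shading logic.
# No chain connects to a material output in the graph.
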Material /VehicleCosmetics/Materials/MAT_Vehicle_Body_MutableContainer
graph {
 properties {bAutomaticallySetUsageInEditor=false, shadingModel=MSM_ClearCoat, usedShadingModels=Clear Coat}
 nodes [
  n1 = Multiply(
   A=Multiply(
    A=VectorParameter() {group=Color, parameterName=BaseColor, sortPriority=1, defaultValue=<rgba 0 0.2 1 1>}.RGB
    B=VectorParameter() {group=Color, parameterName=AccentColor, sortPriority=2, defaultValue=<rgba 1 1 1 1>}.RGB
   )
   B=VectorParameter() {group=Color, parameterName=PaintColor, sortPriority=3, defaultValue=<rgba 0 1 0 1>}.RGB
  )
  n2 = TextureSampleParameter2D() {group=UniversalLogos, parameterName=BodyNormalLOD, samplerType=SAMPLERTYPE_Normal, texture=/VehicleCosmetics/Textures/T_BlankNormal}
  n3 = Multiply(A=Multiply(A=@n1, B=ScalarParameter() {parameterName=LOD}), B=@n2.RGB)
  n4 = VectorParameter() {
   desc="RG=XYScale BA=XYOffset"
   group=UniversalLogos
   parameterName=TrimSheet1_Transform
   sortPriority=4
   channelNames {a=Y Offset, b=X Offset, g=Y Scale, r=X Scale}
   defaultValue=<rgba 2 1.25 0 0.69>
  }
  n5 = Multiply(
   A=Multiply(A=Multiply(A=@n3, B=@n4), B=ScalarParameter() {desc=Degrees counterclockwise, group=UniversalLogos, parameterName=TrimSheet1_Rotation})
   B=ScalarParameter() {desc="Enables the mirrored TrimSheet; 0 = hidden; 1 = visible", group=UniversalLogos, parameterName=TrimSheet1Mirrored_Opacity}
  )
  n6 = VectorParameter() {
   desc="RG=XY Scale Offset; BA=XY Position Offset"
   group=UniversalLogos
   parameterName=TrimSheet1Mirrored_Offset
   sortPriority=4
   channelNames {a=Y Offset, b=X Offset, g=Y Scale, r=X Scale}
  }
  n7 = TextureSampleParameter2D() {group=BaseTextures, parameterName=BodyTextures, samplerSource=SSM_Wrap_WorldGroupSettings, texture=/VehicleCosmetics/Textures/T_Black}
  n8 = Multiply(
   A=Multiply(
    A=Multiply(A=@n5, B=@n6)
    B=ScalarParameter() {desc=Offset by degrees counterclockwise, group=UniversalLogos, parameterName=TrimSheet1Mirrored_RotationOffset}
   )
   B=@n7.RGB
  )
  n9 = TextureSampleParameter2D() {
   group=BaseTextures
   parameterName=Ubershader_Normal
   samplerSource=SSM_Wrap_WorldGroupSettings
   samplerType=SAMPLERTYPE_LinearColor
   texture=/VehicleCosmetics/Textures/T_Black_Linear
  }
  Multiply(B=Multiply(
    A=Multiply(A=@n8, B=@n9.RGB)
    B=VectorParameter() {group=Color, parameterName=EmissiveColor, sortPriority=5, defaultValue=<rgba 0 1 1 0>}.RGB
   ))
  Multiply(A=@n1, B=VectorParameter() {group=Color, parameterName=TrimColor, sortPriority=4, defaultValue=<rgba 0.03 0.03 0.03 0>}.RGB)
  Multiply()
  n10 = TextureSampleParameter2D() {group=UniversalLogos, parameterName=TrimSheet, samplerType=SAMPLERTYPE_Normal, texture=&T_BlankNormal}
  Multiply(A=@n3, B=@n10.RGB)
 ]
}
