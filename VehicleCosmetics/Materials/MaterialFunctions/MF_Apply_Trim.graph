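# Material function for vehicle cosmetics that applies the trim material over an incoming set of material attributes.
# It blends the input attributes with the output of MF_Trim, using the INTrimMask input as alpha.
# After the blend it writes the incoming Normal back into both Normal and ClearCoatBottomNormal, so the base surface normal overrides any from the trim.
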MaterialFunction /VehicleCosmetics/Materials/MaterialFunctions/MF_Apply_Trim
graph {nodes [
  n1 = GetMaterialAttributes(None=FunctionInput() {inputName=MaterialAttributes, inputType=FunctionInput_MaterialAttributes}) {attributeGetTypes [#1, #2]}
  n2 = GetMaterialAttributes(None=MaterialFunctionCall() {materialFunction=&/MF_Trim}.Result)
  n3 = FunctionInput() {inputName=INTrimMask, sortPriority=1}
  FunctionOutput(None=SetMaterialAttributes!(
    MaterialAttributes=BlendMaterialAttributes(A=@n1.MaterialAttributes, B=@n2.MaterialAttributes, Alpha=@n3)
    Normal=@n1.Normal
    ClearCoatBottomNormal=@n1.Normal
   ) {attributeSetTypes [#1, #2]}) {outputName=SetMaterialAttributes}
 ]}
