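# Base material function for vehicle chassis cosmetics, outputting clear-coat material attributes plus trim, carbon fiber and plastics masks.
# Packed textures supply paint masking, curvature overlay, AO, roughness and metallic, with blackout regions derived from the AO channel.
# Headlight, brake, reverse, boost and glowing-part emissives combine masks, colors and brightness with Blueprint-only state scalars.
# Static switches allow cropping the emissive mask texture and letting base color tint emissives.
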MaterialFunction /VehicleCosmetics/Materials/MaterialFunctions/MF_Chassis_Base
graph {nodes [
  n1 = NamedRerouteUsage()
  Add(B=Multiply(A=NamedRerouteUsage(), B=@n1))
  n2 = VectorParameter() {group=Emissive, parameterName=EmissiveUVs_RG_UpperLeftCorner_BA_LowerRightCorner, defaultValue=<rgba 0 0 1 1>}
  n3 = AppendVector(A=@n2.RGB, B=@n2.RGB)
  n4 = TextureObjectParameter() {desc=_E, group=Textures, parameterName=EmissiveMasks, samplerType=SAMPLERTYPE_LinearColor, texture=/VehicleCosmetics/Textures/T_Black_Linear}
  Composite!(RGBA=@n3, T2D_In=@n4)
  Composite!()
  n5 = TextureSampleParameter2D() {
   group=Textures
   parameterName=ChassisComposite_Mask_LOD
   samplerSource=SSM_Wrap_WorldGroupSettings
   samplerType=SAMPLERTYPE_LinearColor
   texture=&T_Black_Linear
  }
  FunctionOutput(None=NamedRerouteDeclaration(None=@n5.G) {name=TrimMask, nodeColor=<rgba 0.026 0.026 0.026 1>}) {outputName=TrimMask}
  n6 = Add(A=NamedRerouteUsage(), B=LinearInterpolate(A=Constant() {r=-1}, B=Constant() {r=1}, Alpha=Constant() {r=0.55}))
  n7 = NamedRerouteUsage()
  n8 = NamedRerouteUsage()
  n9 = OneMinus(None=NamedRerouteUsage())
  n10 = SetMaterialAttributes!(
   MaterialAttributes
   BaseColor=NamedRerouteUsage()
   Metallic=LinearInterpolate(A=NamedRerouteUsage(), Alpha=NamedRerouteUsage())
   Roughness=Clamp(None=LinearInterpolate(A=@n6, Alpha=@n7) {desc=BlackOut})
   EmissiveColor=NamedRerouteUsage()
   Normal=@n8
   AmbientOcclusion=NamedRerouteUsage()
   CustomData0=Multiply(A=@n9, B=ScalarParameter() {group=Chassis, parameterName=Chassis_ClearCoat}) {desc=BlackOut}
   CustomData1=LinearInterpolate(B=ScalarParameter() {group=Chassis, parameterName=Chassis_ClearCoatRoughness}, Alpha=@n9) {constA=1}
   ClearCoatBottomNormal=@n8
   ShadingModel=ShadingModel() {shadingModel=MSM_ClearCoat}
  ) {attributeSetTypes [
    #1
    #2
    #3
    #4
    #5
    #6
    #7
    #8
    #9
    #10
   ]}
  FunctionOutput(None=@n10) {outputName=SetMaterialAttributes}
  FunctionOutput(None=NamedRerouteDeclaration(None=@n5.R) {name=CarbonFiberMask, nodeColor=<rgba 0.026 0.026 0.026 1>}) {outputName=CarbonFiberMask}
  FunctionOutput(None=NamedRerouteDeclaration(None=@n5.B) {name=Plastics, nodeColor=<rgba 0.026 0.026 0.026 1>}) {outputName=PlasticsMask}
  n11 = NamedRerouteUsage()
  n12 = NamedRerouteUsage()
  n13 = Multiply(A=@n11, B=@n12)
  n14 = NamedRerouteUsage()
  n15 = Multiply(A=NamedRerouteUsage(), B=@n14)
  n16 = NamedRerouteUsage()
  n17 = NamedRerouteUsage()
  n18 = NamedRerouteUsage()
  n19 = Add(A=Add(A=Add(A=@n13, B=@n15), B=Multiply(A=@n16, B=@n17)), B=Multiply(A=@n11, B=@n18))
  MaterialExpressionReroute_2 = Reroute(None=@n19)
  n20 = Max(A=Max(A=@n12, B=@n14), B=@n17)
  n21 = Saturate(None=ConstantBiasScale(None=Max(A=@n20, B=@n18)) {bias=0, scale=2})
  MaterialExpressionReroute_3 = Reroute(None=@n21)
  n22 = LinearInterpolate(A=NamedRerouteUsage(), B=@MaterialExpressionReroute_2.EmissiveColors, Alpha=@MaterialExpressionReroute_3.EmissiveMasks)
  LinearInterpolate(
   A=LinearInterpolate(A=@n22, Alpha=NamedRerouteUsage()) {constB=0}
   B=@n22
   Alpha=Saturate(None=ScalarParameter() {group=BP-USE-ONLY, parameterName=BoostGlowIntensity})
  )
  Max(B=@n1)
  n23 = TextureSampleParameter2D() {
   desc=_M
   group=Textures
   parameterName=Masks
   samplerSource=SSM_Wrap_WorldGroupSettings
   samplerType=SAMPLERTYPE_LinearColor
   sortPriority=33
   texture=&T_Black_Linear
   channelNames {a="", b=BoostMask, g=PaintMask, r=EmssivePartsMask}
  }
  NamedRerouteDeclaration(None=@n23.PaintMask) {name=ChassisPaintedMask, nodeColor=<rgba 0 0.1726 0.4792 1>}
  NamedRerouteDeclaration(None=VectorParameter() {group=Color, parameterName=Chassis_HeadlightColor, defaultValue=<rgba 0.6249 1 0.9901 1>}.RGB) {name=HeadlightColor, nodeColor=<rgba 0.9282 0 1 1>}
  n24 = TextureSampleParameter2D() {
   group=Textures
   parameterName=Normal
   samplerSource=SSM_Wrap_WorldGroupSettings
   samplerType=SAMPLERTYPE_Normal
   sortPriority=34
   texture=/VehicleCosmetics/Textures/T_BlankNormal_B
  }
  NamedRerouteDeclaration(None=@n24.RGB) {name=NormalTexture, nodeColor=<rgba 0.6038 0.3325 0.013 1>}
  NamedRerouteDeclaration(None=MaterialFunctionCall!(
    "Rainbow Lerp 0=Default Color 1=Rainbow"
    Default Color=VectorParameter() {group=Color, parameterName=Chassis_BoostColor, defaultValue=<rgba 1 0.05 0 1>}.RGB
    UVs
    Panner Speed Panner X
   ) {materialFunction=/DealershipSharedLibrary/FX/Materials/MaterialFunctions/MF_VCVFX_RainbowColor}.Result) {name=BoostColor, nodeColor=<rgba 0.9282 0 1 1>}
  n25 = VectorParameter() {
   desc=Brightness Settings
   group=Emissive
   parameterName=Chassis_EmissiveBrightness
   sortPriority=0
   channelNames {a=BoostEmissive, b=BrakeLightEmissive, g=HeadLightEmissive, r=ChassisEmissiveParts}
   defaultValue=<rgba 1 1 1 1>
  }
  NamedRerouteDeclaration(None=@n25.ChassisEmissiveParts) {name=ChassisEmissivePartsBrightness, nodeColor=<rgba 0.0436 1 0 1>}
  NamedRerouteDeclaration(None=@n25.HeadLightEmissive) {name=HeadlightEmissiveBrightness, nodeColor=<rgba 0.0436 1 0 1>}
  NamedRerouteDeclaration(None=@n25.BrakeLightEmissive) {name=BrakelightEmissiveBrightness, nodeColor=<rgba 0.0436 1 0 1>}
  NamedRerouteDeclaration(None=@n25.BoostEmissive) {name=BoostEmissiveBrightness, nodeColor=<rgba 0.0436 1 0 1>}
  NamedRerouteDeclaration(None=ScalarParameter() {group=BP-USE-ONLY, parameterName=BoostGlowIntensity}) {name=BoostGlowVariable, nodeColor=<rgba 0 0 0 1>}
  NamedRerouteDeclaration(None=ScalarParameter() {group=BP-USE-ONLY, parameterName=isBraking}) {name=isBrakingVariable, nodeColor=<rgba 0 0 0 1>}
  NamedRerouteDeclaration(None=ScalarParameter() {group=BP-USE-ONLY, parameterName=ReverseLight}) {name=ReverseLightVariable, nodeColor=<rgba 0 0 0 1>}
  NamedRerouteDeclaration(None=ScalarParameter() {group=BP-USE-ONLY, parameterName=HeadlightsFullyOn}) {name=HeadlightsFullyOnVariable, nodeColor=<rgba 0 0 0 1>}
  n26 = TextureSampleParameter2D() {
   group=Textures
   parameterName=BaseColor
   samplerSource=SSM_Wrap_WorldGroupSettings
   sortPriority=30
   texture=/VehicleCosmetics/Textures/T_Black
  }
  NamedRerouteDeclaration(None=@n26.RGB) {name=BaseColorTexture, nodeColor=<rgba 0.6038 0.3325 0.013 1>}
  NamedRerouteDeclaration(None=@n23.BoostMask) {name=ChassisBoostEmissiveMask, nodeColor=<rgba 0 0.1726 0.4792 1>}
  MaterialExpressionReroute_1 = Reroute(None=@n3)
  MaterialExpressionReroute_9 = Reroute(None=@n4)
  n27 = StaticSwitchParameter(
   True=MaterialFunctionCall!(
    UVs
    TextureIn=@MaterialExpressionReroute_9.T2D_In
    UpperLeftCorner=ComponentMask(None=@MaterialExpressionReroute_1.RGBA) {g=true, r=true}
    LowerRightCorner=ComponentMask(None=@MaterialExpressionReroute_1.RGBA) {a=true, b=true}
    ExportFloat4
   ) {materialFunction=/Engine/Functions/Engine_MaterialFunctions01/Texturing/TextureCropping}.CroppedMasked
   False=TextureSample(Tex=@MaterialExpressionReroute_9.T2D_In) {samplerType=SAMPLERTYPE_LinearColor, texture=&T_Black_Linear}.RGBA
  ) {desc=CROP_EMIS, group=Emissive, parameterName=CroppedEmissive}
  n28 = ComponentMask(None=@n27) {g=true}
  MaterialExpressionReroute_13 = Reroute(None=@n28)
  NamedRerouteDeclaration(None=@MaterialExpressionReroute_13.HeadLight) {name=HeadlightEmissiveMask, nodeColor=<rgba 0 0.1726 0.4792 1>}
  n29 = ComponentMask(None=@n27) {b=true}
  MaterialExpressionReroute_14 = Reroute(None=@n29)
  NamedRerouteDeclaration(None=@MaterialExpressionReroute_14.BrakeLight) {name=BrakeLightEmissiveMask, nodeColor=<rgba 0 0.1726 0.4792 1>}
  n30 = ComponentMask(None=@n27) {r=true}
  MaterialExpressionReroute_0 = Reroute(None=@n30)
  NamedRerouteDeclaration(None=@MaterialExpressionReroute_0.ReverseLight) {name=ReverseLightEmissiveMask, nodeColor=<rgba 0 0.1726 0.4792 1>}
  n31 = NamedRerouteUsage()
  n32 = Multiply(A=NamedRerouteUsage(), B=Multiply(A=@n31, B=NamedRerouteUsage())) {desc=HeadLight Color and Brightness}
  n33 = Multiply(A=Saturate(None=ConstantBiasScale(None=@n31) {bias=0.1, scale=1.5}), B=@n32) {desc="Enhances BC emissive so it doesn't fade to white.\r\nDoes slightly affect Headlight Color if it's enabled."}
  n34 = StaticSwitchParameter(True=Constant() {r=1}, False=Constant()) {desc=BC overrides emissive color while true., group=Color, parameterName=UseBC_HeadlightEmissive}
  NamedRerouteDeclaration(None=Multiply(A=NamedRerouteUsage(), B=LinearInterpolate(A=@n32, B=@n33, Alpha=@n34) {desc="Push color of BC without\r\naffecting any other settings."}) {desc=HeadLight Variable Mask}) {name=HeadLightCombined, nodeColor=<rgba 1 1 1 1>}
  n35 = NamedRerouteUsage()
  n36 = Multiply(A=NamedRerouteUsage(), B=Multiply(A=@n35, B=NamedRerouteUsage())) {desc=HeadLight Color and Brightness}
  n37 = Multiply(A=Saturate(None=ConstantBiasScale(None=@n35) {bias=0.1, scale=1.5}), B=@n36) {desc="Enhances BC emissive so it doesn't fade to white.\r\nDoes slightly affect Headlight Color if it's enabled."}
  n38 = StaticSwitchParameter(True=Constant() {r=1}, False=Constant()) {desc=BC overrides emissive color while true., group=Color, parameterName=UseBC_BrakeEmissive}
  NamedRerouteDeclaration(None=Multiply(A=NamedRerouteUsage(), B=LinearInterpolate(A=@n36, B=@n37, Alpha=@n38) {desc="Push color of BC without\r\naffecting any other settings."}) {desc=HeadLight Variable Mask}) {name=BrakeLightCombined, nodeColor=<rgba 1 1 1 1>}
  n39 = NamedRerouteUsage()
  n40 = Multiply(A=Constant() {r=1}, B=NamedRerouteUsage())
  NamedRerouteDeclaration(None=Multiply(A=NamedRerouteUsage(), B=Multiply(A=@n39, B=@n40) {desc=HeadLight Color and Brightness})) {name=ReverseLightCombined, nodeColor=<rgba 1 1 1 1>}
  n41 = NamedRerouteUsage()
  n42 = Multiply(A=NamedRerouteUsage(), B=NamedRerouteUsage())
  NamedRerouteDeclaration(None=Multiply(A=NamedRerouteUsage(), B=Multiply(A=@n41, B=@n42) {desc=HeadLight Color and Brightness})) {name=BoostCombined, nodeColor=<rgba 1 1 1 1>}
  NamedRerouteDeclaration(None=@n23.EmssivePartsMask) {name=ChassisEmissivePartsMask, nodeColor=<rgba 0 0.1726 0.4792 1>}
  n43 = VectorParameter() {
   desc="The Glowing parts of the emissive (if any) controlled by the EmissivePartsMask. --"
   group=Color
   parameterName=Chassis_EmissivePartsColor
   defaultValue=<rgba 0.6038 0.3325 0.013 1>
  }
  NamedRerouteDeclaration(None=@n43.RGB) {name=ChassisEmissivePartsColor, nodeColor=<rgba 0.9301 0 1 1>}
  n44 = VectorParameter() {
   desc=The Painted Color controlled by the PaintMask. -- Used for the paintable areas of the chassis --
   group=Color
   parameterName=Chassis_PaintColor
   defaultValue=<rgba 0 0.1434 1 1>
  }
  NamedRerouteDeclaration(None=@n44.RGB) {name=PaintColor, nodeColor=<rgba 0.9282 0 1 1>}
  n45 = NamedRerouteUsage()
  NamedRerouteDeclaration(None=StaticSwitchParameter(True=@n45, False=NamedRerouteUsage()) {desc=BC overrides emissive color while true., group=Color, parameterName=UseBC_BrakeEmissive}) {name=BC_BrakeEmis_Bool, nodeColor=<rgba 0.3763 0.7529 0.7157 1>}
  n46 = NamedRerouteUsage()
  n47 = NamedRerouteUsage()
  n48 = Add(A=Add(A=Add(A=@n46, B=@n47), B=NamedRerouteUsage()), B=NamedRerouteUsage())
  n49 = NamedRerouteUsage()
  NamedRerouteDeclaration(None=Multiply(A=NamedRerouteUsage(), B=Add(A=@n48, B=@n49))) {name=EmissiveColor, nodeColor=<rgba 0 0.7656 1 1>}
  n50 = NamedRerouteUsage()
  n51 = NamedRerouteUsage()
  NamedRerouteDeclaration(None=Multiply(A=NamedRerouteUsage(), B=Multiply(A=@n50, B=@n51))) {name=ChassisEmissivePartsCombined, nodeColor=<rgba 1 1 1 1>}
  NamedRerouteDeclaration(None=StaticSwitchParameter(True=@n45, False=NamedRerouteUsage()) {desc=BC overrides emissive color while true., group=Color, parameterName=UseBC_BoostEmissive}) {name=BC_BoostEmis_Bool, nodeColor=<rgba 0.3763 0.7529 0.7157 1>}
  n52 = NamedRerouteUsage()
  n53 = Multiply(A=Clamp(None=NamedRerouteUsage()), B=@n22) {desc=AO}
  n54 = NamedRerouteUsage()
  NamedRerouteDeclaration(None=Multiply(
    A=OneMinus(None=@n52)
    B=MaterialFunctionCall(Base=@n53, Blend=@n54) {desc=curvature, materialFunction=&Blend_Overlay}.Result
   ) {desc=BlackOut}) {name=BaseColorFinal, nodeColor=<rgba 1 0.6328 0 1>}
  NamedRerouteDeclaration(None=StaticSwitchParameter(True=@n45, False=NamedRerouteUsage()) {desc=BC overrides emissive color while true., group=Color, parameterName=UseBC_HeadlightEmissive}) {name=BC_HeadlightEmiss_Bool, nodeColor=<rgba 0.3763 0.7529 0.7157 1>}
  n55 = TextureSampleParameter2D() {
   group=Textures
   parameterName=ChassisTextures
   samplerSource=SSM_Wrap_WorldGroupSettings
   samplerType=SAMPLERTYPE_LinearColor
   sortPriority=35
   texture=&T_Black_Linear
   channelNames {a=Roughness, b=Metallic, g=AO-BlackOut, r=Curvature}
  }
  NamedRerouteDeclaration(None=@n55.AO-BlackOut) {name=AO, nodeColor=<rgba 0 0.1726 0.4792 1>}
  NamedRerouteDeclaration(None=VectorParameter() {group=Color, parameterName=Brake_Color, defaultValue=<rgba 0.6 0 0 1>}.RGB) {name=ChassisBrakeColor, nodeColor=<rgba 0.9301 0 1 1>}
  NamedRerouteDeclaration(None=StaticSwitchParameter(True=@n45, False=NamedRerouteUsage()) {desc=BC overrides emissive color while true., group=Color, parameterName=UseBC_ChassisPartsEmissive}) {name=BC_ChassisPartsEmis_Bool, nodeColor=<rgba 0.3763 0.7529 0.7157 1>}
  n56 = Ceil(None=NamedRerouteUsage())
  NamedRerouteDeclaration(None=OneMinus(None=Clamp(None=@n56))) {name=Blackout, nodeColor=<rgba 0.5703 1 0 1>}
  NamedRerouteDeclaration(None=LinearInterpolate(A=NamedRerouteUsage(), B=NamedRerouteUsage(), Alpha=NamedRerouteUsage()) {desc=Painted Mask}) {name=BaseColorPainted, nodeColor=<rgba 0 0.7656 1 1>}
  NamedRerouteDeclaration(None=@n55.Curvature) {name=Curvature, nodeColor=<rgba 0 0.1726 0.4792 1>}
  NamedRerouteDeclaration(None=@n55.Roughness) {name=Roughness, nodeColor=<rgba 0 0.1726 0.4792 1>}
  NamedRerouteDeclaration(None=@n55.Metallic) {name=Metallic, nodeColor=<rgba 0 0.1726 0.4792 1>}
  PinBase() {reroutePins [
    RGBA {expression=&self:MaterialExpressionReroute_1}
    T2D_In {expression=&self:MaterialExpressionReroute_9}
   ]}
  PinBase!(ReverseLight=@n30, HeadLight=@n28, BrakeLight=@n29) {reroutePins [
    ReverseLight {expression=&self:MaterialExpressionReroute_0}
    HeadLight {expression=&self:MaterialExpressionReroute_13}
    BrakeLight {expression=&self:MaterialExpressionReroute_14}
   ]}
  PinBase!()
  PinBase!(EmissiveColors=@n19, EmissiveMasks=@n21) {reroutePins [
    EmissiveColors {expression=&self:MaterialExpressionReroute_2}
    EmissiveMasks {expression=&self:MaterialExpressionReroute_3}
   ]}
 ]}
