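# Base Clear Coat material for vehicle chassis cosmetics, built on material attributes and flagged for skeletal meshes and mesh deformers.
# MF_Chassis_Base supplies attributes and trim, carbon fiber and plastics masks, shared through named reroutes.
# Static switches layer Trim, Carbon Fiber and Plastics functions in turn, with an optional scalar-driven trim-to-carbon-fiber blend.
# The result passes through MF_VehicleDeRez_Transition, whose bottom normal feeds the clear coat normal output.
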Material /VehicleCosmetics/Materials/MAT_Vehicle_Chassis_Base
graph {
 properties {
  bUseMaterialAttributes=true
  bUsedWithMeshDeformer=true
  bUsedWithSkeletalMesh=true
  pixelDepthOffsetMode=PDOM_Legacy
  shadingModel=MSM_ClearCoat
  usedShadingModels=Clear Coat
 }
 nodes [
  n1 = NamedRerouteUsage()
  n2 = MaterialFunctionCall() {materialFunction=/VehicleCosmetics/Materials/MaterialFunctions/MF_Chassis_Base}
  n3 = MaterialFunctionCall!(MaterialAttributes=@n2.SetMaterialAttributes, INTrimMask=NamedRerouteUsage()) {materialFunction=/VehicleCosmetics/Materials/MaterialFunctions/MF_Apply_Trim}
  n4 = MaterialFunctionCall(MaterialAttributes=@n3.SetMaterialAttributes, INTrimMask=@n1) {materialFunction=&MF_Apply_Trim}
  n5 = MaterialFunctionCall!(MaterialAttributes=@n3.SetMaterialAttributes, INCarbonFIberMask=@n1) {materialFunction=/VehicleCosmetics/Materials/MaterialFunctions/MF_Apply_CarbonFiber}
  n6 = ScalarParameter() {parameterName=Trim_SwapTrimForCarbonFiber}
  n7 = StaticSwitchParameter(
   True=StaticSwitchParameter(True=BlendMaterialAttributes(A=@n4.SetMaterialAttributes, B=@n5.SetMaterialAttributes, Alpha=@n6), False=@n3.SetMaterialAttributes) {group=SpecialFeatures, parameterName=TrimSwapForCarbonFiber}
   False=@n2.SetMaterialAttributes
  ) {group=Trim, parameterName=ApplyTrim}
  n8 = StaticSwitchParameter(
   True=MaterialFunctionCall(MaterialAttributes=@n7, INCarbonFIberMask=NamedRerouteUsage()) {materialFunction=&MF_Apply_CarbonFiber}.SetMaterialAttributes
   False=@n7
  ) {group=CarbonFiber, parameterName=ApplyCarbonFiber}
  n9 = StaticSwitchParameter(
   True=MaterialFunctionCall!(MaterialAttributes=@n8, INPlasticMask=NamedRerouteUsage()) {materialFunction=/VehicleCosmetics/Materials/MaterialFunctions/MF_Apply_Plastics}.SetMaterialAttributes
   False=@n8
  ) {group=Plastic, parameterName=ApplyPlastics}
  ClearCoatNormalCustomOutput!(None=GetMaterialAttributes(None=MaterialFunctionCall(In=@n9) {materialFunction=&MF_VehicleDeRez_Transition}.Result) {attributeGetTypes [#1]}.ClearCoatBottomNormal)
  Constant() {r=1}
  NamedRerouteDeclaration(None=@n2.TrimMask) {name=ChassisBase_TrimMask, nodeColor=<rgba 0 1 0.1797 1>}
  NamedRerouteDeclaration(None=@n2.CarbonFiberMask) {name=ChassisBase_CarbonFiberMask, nodeColor=<rgba 0.6641 0 1 1>}
  NamedRerouteDeclaration(None=@n2.PlasticsMask) {name=ChassisBase_PlasticsMask, nodeColor=<rgba 1 0 0.6797 1>}
  SetMaterialAttributes!(
   MaterialAttributes=@n5.SetMaterialAttributes
   ClearCoatBottomNormal=GetMaterialAttributes(None=@n3.SetMaterialAttributes) {attributeGetTypes [#1]}.ClearCoatBottomNormal
  ) {attributeSetTypes [#1]}
 ]
}
