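# Masked, default-lit master material for vehicle decals, flagged for skeletal meshes and mesh deformers.
# One Tiling parameter scales the UVs for Base_Color, Normal and a packed Masks texture holding roughness, opacity, metallic and emissive.
# Normal_Intensity multiplies the normal sample, while Use_Emissive, color and multiplier parameters scale the emissive mask.
# A separate Ceil and Saturate branch turns Use_Emissive into a clamped, rounded-up switch value.
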Material /VehicleCosmetics/Materials/MAT_Vehicle_Decals
graph {
 properties {
  bAutomaticallySetUsageInEditor=false
  bUsedWithMeshDeformer=true
  bUsedWithSkeletalMesh=true
  blendMode=BLEND_Masked
  usedShadingModels=Default Lit
 }
 nodes [
  n1 = Multiply(A=TextureCoordinate(), B=ScalarParameter() {defaultValue=1, parameterName=Tiling})
  Multiply(
   A=TextureSampleParameter2D(UVs=@n1) {parameterName=Normal, samplerType=SAMPLERTYPE_Normal, texture=/VehicleCosmetics/Textures/T_BlankNormal_B}.RGB
   B=VectorParameter() {parameterName=Normal_Intensity, defaultValue=<rgba 2 2 1 1>}.RGB
  )
  n2 = ScalarParameter() {group=Emissive, parameterName=Use_Emissive}
  n3 = TextureSampleParameter2D(UVs=@n1) {
   parameterName=Masks
   samplerType=SAMPLERTYPE_LinearColor
   texture=/VehicleCosmetics/Textures/T_Black_Linear
   channelNames {a=Emissive, b=Metallic, g=Opacity, r=Roughness}
  }
  Multiply(
   A=Multiply(
    A=Multiply(A=@n2, B=@n3.Emissive)
    B=VectorParameter() {group=Emissive, parameterName=Emissive_Color, defaultValue=<rgba 0.745 0.9519 1 1>}.RGB
   )
   B=ScalarParameter() {defaultValue=1, group=Emissive, parameterName=Emissive_Multiplier}
  )
  Saturate(None=Ceil(None=@n2))
  TextureSampleParameter2D(UVs=@n1) {parameterName=Base_Color, texture=/VehicleCosmetics/Textures/T_White}
 ]
}
