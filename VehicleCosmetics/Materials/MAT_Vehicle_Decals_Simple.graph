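# Masked, Default Lit material for simple vehicle decals, flagged for use on skeletal meshes and mesh deformers.
# A scalar Tiling parameter multiplies the UVs shared by three texture parameters: Albedo, Normal and Masks.
# Masks is a linear-color texture with channels labeled blue for Metallic and alpha for Roughness, leaving red and green unused.
# Neutral white, blank-normal and black defaults act as placeholders for overriding textures.
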Material /VehicleCosmetics/Materials/MAT_Vehicle_Decals_Simple
graph {
 properties {
  bAutomaticallySetUsageInEditor=false
  bUsedWithMeshDeformer=true
  bUsedWithSkeletalMesh=true
  blendMode=BLEND_Masked
  usedShadingModels=Default Lit
 }
 nodes [
  n1 = Multiply(A=TextureCoordinate(), B=ScalarParameter() {defaultValue=1, parameterName=Tiling})
  TextureSampleParameter2D(UVs=@n1) {parameterName=Albedo, texture=/VehicleCosmetics/Textures/T_White}
  TextureSampleParameter2D(UVs=@n1) {parameterName=Normal, samplerType=SAMPLERTYPE_Normal, texture=/VehicleCosmetics/Textures/T_BlankNormal_B}
  TextureSampleParameter2D(UVs=@n1) {
   parameterName=Masks
   samplerType=SAMPLERTYPE_LinearColor
   texture=/VehicleCosmetics/Textures/T_Black_Linear
   channelNames {a=Roughness, b=Metallic, g=none, r=none}
  }
 ]
}
