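# Parent material for Huntress vehicle body skins, giving a two-color Default Lit surface with an infinity depth effect.
# The Huntress RGB map runs through two infinity offset functions, whose maxed R and G channels are tinted by AccentColor and BaseColor.
# Both layers are masked to black, with B declared as the infinity mask, then summed.
# A UseEmissive switch adds optional glow, and the attributes pass through the shared vehicle PostFX function.
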Material /VehicleCosmetics/Materials/MAT_Vehicle_Huntress_Parent
graph {
 properties {
  bUseMaterialAttributes=true
  bUsedWithMeshDeformer=true
  bUsedWithSkeletalMesh=true
  pixelDepthOffsetMode=PDOM_Legacy
  usedShadingModels=Default Lit
 }
 nodes [
  MaterialFunctionCall!(In=MakeMaterialAttributes!(
    BaseColor=NamedRerouteUsage()
    Metallic=ScalarParameter() {defaultValue=0.5, group=Metallic, parameterName=Metallic, sliderMax=1, sliderMin=0.01}
    Specular=ScalarParameter() {desc=Body, group=Specular, parameterName=Specular}
    Roughness=ScalarParameter() {defaultValue=0.35, desc=Body, group=Roughness, parameterName=Roughness}
    Anisotropy
    EmissiveColor=NamedRerouteUsage()
    Opacity
    OpacityMask
    Normal
    Tangent
    WorldPositionOffset
    SubsurfaceColor
    ClearCoat
    ClearCoatRoughness
    AmbientOcclusion
    Refraction
    CustomizedUVs_0
    CustomizedUVs_1
    CustomizedUVs_2
    CustomizedUVs_3
    CustomizedUVs_4
    CustomizedUVs_5
    CustomizedUVs_6
    CustomizedUVs_7
    PixelDepthOffset
    ShadingModel
    Displacement
   )) {materialFunction=/VehicleCosmetics/Materials/MaterialFunctions/MF_Vehicle_FNPostFX}
  n1 = TextureObjectParameter() {
   desc="R = Custom Color, G = Team Color, B = Mask Out Infinity Effect"
   group=Huntress_Params
   parameterName=Huntress_Map
   texture=/VehicleCosmetics/Bodies/Octane/Skins/Skin_OctaneHuntress_RGB
  }
  NamedRerouteDeclaration(None=TextureSample(Tex=@n1) {texture=&Skin_OctaneHuntress_RGB}.B) {name=TextureMask_B, nodeColor=<rgba 1 0 0.3047 1>}
  n2 = Constant() {r=0.0012}
  n3 = MaterialFunctionCall(Float2=MaterialFunctionCall!(Texture=@n1, Offset=Constant() {r=0.5}, RotateUVs, GradientDarkness=@n2, GradientFalloff, UVs, Opt_LowQualityOffset, Opt_MediumQualityOffset) {materialFunction=/VehicleCosmetics/SharedMaterials/MaterialFunctions/MF_Infinity_OffsetFunction}.Result) {materialFunction=&BreakOutFloat2Components}
  n4 = MaterialFunctionCall(Float2=MaterialFunctionCall(Texture=@n1, Offset=ScalarParameter() {defaultValue=0.33, parameterName=BumpOffsetAmount}, GradientDarkness=@n2) {materialFunction=&MF_Infinity_OffsetFunction}.Result) {materialFunction=&BreakOutFloat2Components}
  n5 = Multiply(
   A=Multiply(A=VectorParameter() {group=Color, parameterName=AccentColor, sortPriority=2, defaultValue=<rgba 1 1 1 0>}.RGB) {constB=2}
   B=Saturate(None=Max(A=@n3.R, B=@n4.R))
  )
  n6 = NamedRerouteUsage()
  n7 = Multiply(
   A=Saturate(None=Max(A=@n3.G, B=@n4.G))
   B=Multiply(A=VectorParameter() {group=Color, parameterName=BaseColor, sortPriority=1, defaultValue=<rgba 0 0.2 1 1>}.RGB) {constA=1.75, constB=1.75}
  )
  NamedRerouteDeclaration(None=Add(A=LinearInterpolate(A=@n5, Alpha=@n6) {constB=0}, B=LinearInterpolate(A=@n7, Alpha=@n6) {constB=0})) {name=BC_Out, nodeColor=<rgba 0.9453 1 0 1>}
  n8 = NamedRerouteUsage()
  NamedRerouteDeclaration(None=StaticSwitchParameter(True=Multiply(A=@n8) {constB=2}, False=Constant()) {parameterName=UseEmissive}) {name=Emissive_Out, nodeColor=<rgba 0.7344 0 1 1>}
 ]
}
