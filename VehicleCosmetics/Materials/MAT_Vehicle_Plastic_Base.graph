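# Base material for plastic surfaces on vehicle cosmetics, usable on skeletal meshes and mesh deformers.
# MF_Plastic_Base builds the material attributes, with shading model chosen by expression between Default Lit and Clear Coat.
# Its result feeds MF_Vehicle_FNPostFX and, on a separate branch, MF_BaseVehicle_Damage followed by MF_VehicleDeRez_Transition.
# The graph holds only these function calls, so all surface and effect logic lives in the material functions.
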Material /VehicleCosmetics/Materials/MAT_Vehicle_Plastic_Base
graph {
 properties {
  bUseMaterialAttributes=true
  bUsedWithMeshDeformer=true
  bUsedWithSkeletalMesh=true
  pixelDepthOffsetMode=PDOM_Legacy
  shadingModel=MSM_FromMaterialExpression
  usedShadingModels=Default Lit | Clear Coat
 }
 nodes [
  n1 = MaterialFunctionCall() {materialFunction=/VehicleCosmetics/Materials/MaterialFunctions/MF_Plastic_Base}
  MaterialFunctionCall!(In=@n1.Result) {materialFunction=/VehicleCosmetics/Materials/MaterialFunctions/MF_Vehicle_FNPostFX}
  MaterialFunctionCall(In=MaterialFunctionCall(Material=@n1.Result) {materialFunction=&MF_BaseVehicle_Damage}.Result) {materialFunction=&MF_VehicleDeRez_Transition}
 ]
}
