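# Clear coat material for vehicle trim with livery, flagged for use on skeletal meshes and with mesh deformers.
# MF_Trim_Livery supplies the base material attributes, which pass through MF_BaseVehicle_Damage and then MF_VehicleDeRez_Transition.
# All shading logic lives in these three material functions, so the graph itself only chains livery, damage and de-rez transition layers.
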Material /VehicleCosmetics/Materials/MAT_Vehicle_Trim_Livery
graph {
 properties {
  bUseMaterialAttributes=true
  bUsedWithMeshDeformer=true
  bUsedWithSkeletalMesh=true
  pixelDepthOffsetMode=PDOM_Legacy
  shadingModel=MSM_ClearCoat
  usedShadingModels=Clear Coat
 }
 nodes [MaterialFunctionCall(In=MaterialFunctionCall(Material=MaterialFunctionCall() {materialFunction=/VehicleCosmetics/Materials/MaterialFunctions/MF_Trim_Livery}.TrimLivery_MA) {materialFunction=&MF_BaseVehicle_Damage}.Result) {materialFunction=&MF_VehicleDeRez_Transition}]
}
