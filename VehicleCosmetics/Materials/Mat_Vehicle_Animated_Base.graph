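# Switch-driven base material for animated, skinned vehicle cosmetics, recolored through paint, team, accent or custom color parameters.
# Two panning decal layers, blended by ramp, noise or height, set base color and mix two detail normal maps.
# Two self-distorting noise layers add faded emissive FX, with optional pixelated UVs, outlines, drop shadow and pearlescence.
# Body-texture AO darkens the result, and all attributes pass through CharacterDeRez_Transition.
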Material /VehicleCosmetics/Materials/Mat_Vehicle_Animated_Base
graph {
 properties {
  bUseMaterialAttributes=true
  bUsedWithMeshDeformer=true
  bUsedWithSkeletalMesh=true
  pixelDepthOffsetMode=PDOM_Legacy
  usedShadingModels=Default Lit
 }
 nodes [
  n1 = Constant() {r=1}
  n2 = VectorParameter() {group=Colors, parameterName=PaintColor, defaultValue=<rgba 1 0 0.0034 1>}
  n3 = Constant() {r=0.6}
  n4 = Multiply(A=@n3, B=VectorParameter() {group=Colors, parameterName=AccentColor, defaultValue=<rgba 1 1 1 0>}.RGB)
  n5 = VectorParameter() {desc="Alpha = Emissive Color Brightness", group=FX, parameterName=FX*Color2, defaultValue=<rgba 1 0 0.0034 1>}
  n6 = StaticSwitchParameter(True=StaticSwitchParameter(True=@n2.RGB, False=@n4) {group=FX, parameterName=UsePaintColor_FX_2}, False=@n5.RGB) {group=Colors, parameterName=Use2Colors}
  n7 = VectorParameter() {desc="Alpha = Emissive Color Brightness", group=Colors, parameterName=TeamColor, defaultValue=<rgba 0 0.1715 0.4806 1>}
  n8 = Multiply(A=@n3, B=@n7.RGB)
  n9 = VectorParameter() {desc="Alpha = Emissive Color Brightness", group=FX, parameterName=FX*Color1, defaultValue=<rgba 1 0.6333 0.0563 1>}
  n10 = StaticSwitchParameter(True=StaticSwitchParameter(True=@n2.RGB, False=@n8) {group=FX, parameterName=UsePaintColor_FX_1}, False=@n9.RGB) {group=Colors, parameterName=Use2Colors}
  n11 = StaticSwitchParameter(True=@n6, False=@n10) {group=Colors, parameterName=SwapFXColors}
  n12 = StaticSwitchParameter(True=@n10, False=@n6) {group=Colors, parameterName=SwapFXColors}
  n13 = TextureSampleParameter2D() {
   group=BaseTextures
   parameterName=Ubershader_Normal
   samplerSource=SSM_Wrap_WorldGroupSettings
   samplerType=SAMPLERTYPE_LinearColor
   texture=/VehicleCosmetics/Textures/T_Black_Linear
  }
  n14 = TextureCoordinate() {coordinateIndex=1}
  n15 = ScalarParameter() {group=Decals, parameterName=FXFadeRotation}
  n16 = LinearInterpolate(A=@n14, B=AppendVector(A=ComponentMask(None=@n14) {r=true}, B=ComponentMask(None=@n14) {g=true}), Alpha=@n15)
  n17 = Rotator(Coordinate=@n14, Time=@n15) {speed=1}
  n18 = StaticSwitchParameter(
   True=@n13.B
   False=ComponentMask(None=StaticSwitchParameter(True=@n16, False=@n17) {defaultValue=true, group=FX, parameterName=FXFadeRampCheapRotation}) {g=true}
  ) {group=FX, parameterName=FXFade*UseEdgeDist}
  n19 = ScalarParameter() {defaultValue=0.75, group=FX, parameterName=FXFadePosition}
  n20 = ScalarParameter() {defaultValue=3, group=FX, parameterName=FXFadeWidth}
  n21 = StaticSwitchParameter(True=TextureCoordinate() {coordinateIndex=1}, False=TextureCoordinate()) {defaultValue=true, group=FX, parameterName=FX*UseCylinderUV}
  n22 = StaticSwitchParameter(True=AppendVector(A=ComponentMask(None=@n21) {g=true}, B=ComponentMask(None=@n21) {r=true}), False=@n21) {group=FX, parameterName=FX*NoiseSwapUV}
  n23 = ComponentMask(None=@n22) {g=true}
  n24 = ScalarParameter() {defaultValue=0.125, desc="Replaces  linear coords with body normal map X,Y channels.", group=FX, parameterName=FXDistortByEdgeDistance}
  n25 = StaticSwitchParameter(True=AppendVector(A=ComponentMask(None=@n22) {r=true}, B=LinearInterpolate(A=@n23, B=@n13.B, Alpha=@n24)), False=@n22) {group=FX, parameterName=FXDistortByEdgeDistance}
  n26 = AppendVector(A=LinearInterpolate(Alpha=@n13.R) {constA=-1, desc=R}, B=LinearInterpolate(Alpha=@n13.G) {constA=-1, desc=G}) {desc=R+G}
  n27 = ScalarParameter() {defaultValue=0.125, desc="Replaces  linear coords with body normal map X,Y channels.", group=FX, parameterName=FXDistortByBodyNormal}
  n28 = StaticSwitchParameter(True=LinearInterpolate(A=@n25, B=@n26, Alpha=@n27), False=@n25) {group=FX, parameterName=FXDistortByBodyNormal}
  n29 = Constant3Vector() {constant=<rgba 0 0 1 0>}
  n30 = Time() {bOverride_Period=true, period=20}
  n31 = VectorParameter() {desc="(R,G) = Detail 1 Pan (X,Y)    (B,A) = Detail 2 Pan (X,Y)", group=Detail, parameterName=DetailTransformPanner}
  n32 = AppendVector(A=@n31.R, B=@n31.R)
  n33 = StaticSwitchParameter(True=TextureCoordinate() {coordinateIndex=1, uTiling=6, vTiling=6}, False=TextureCoordinate() {uTiling=6, vTiling=6}) {defaultValue=true, group=Detail, parameterName=Detail1UseCylinderUV}
  n34 = AppendVector(A=ComponentMask(None=@n33) {g=true}, B=ComponentMask(None=@n33) {r=true})
  n35 = VectorParameter() {desc="(R,G) = Detail1    (B,A) = Detail2", group=Detail, parameterName=DetailTransform, defaultValue=<rgba 1 1 1 1>}
  n36 = Rotator(
   Coordinate=Multiply(
    A=StaticSwitchParameter(True=@n34, False=@n33) {defaultValue=true, group=Detail, parameterName=Detail1SwapXY}
    B=ComponentMask(None=@n35.RGB) {g=true, r=true}
   )
   Time=ScalarParameter() {desc="Rotate Detail1 (0.25 = 90, 0.5 = 180, 0.75 = 270, 1=360)", group=Detail, parameterName=Detail1Rotation}
  ) {speed=6.2832}
  n37 = Time() {bOverride_Period=true, period=20}
  n38 = Multiply(
   A=@n37
   B=ScalarParameter() {desc="Detail1 Animated Spin Speed (Pivot point isn't very predictable)", group=Detail, parameterName=Detail1Spinner}
  )
  n39 = TextureSampleParameter2D(UVs=Add(A=Multiply(A=@n30, B=@n32), B=Rotator(Coordinate=@n36, Time=@n38) {speed=6.2832}) {desc=Panner}) {group=Detail, parameterName=Detail1, samplerSource=SSM_Wrap_WorldGroupSettings, texture=/VehicleCosmetics/Textures/T_Black}
  n40 = LinearInterpolate(A=@n29, B=@n39.RGB, Alpha=Saturate(None=ScalarParameter() {defaultValue=0.25, group=Detail, parameterName=Detail1Strength}))
  n41 = AppendVector(A=@n31.B, B=@n31.B)
  n42 = StaticSwitchParameter(True=TextureCoordinate() {coordinateIndex=1, uTiling=6, vTiling=6}, False=TextureCoordinate() {uTiling=6, vTiling=6}) {defaultValue=true, group=Detail, parameterName=Detail2UseCylinderUV}
  n43 = AppendVector(A=ComponentMask(None=@n42) {g=true}, B=ComponentMask(None=@n42) {r=true})
  n44 = Rotator(
   Coordinate=Multiply(
    A=StaticSwitchParameter(True=@n43, False=@n42) {defaultValue=true, group=Detail, parameterName=Detail2SwapXY}
    B=AppendVector(A=@n35.B, B=@n35.B)
   )
   Time=ScalarParameter() {desc="Rotate Detail2 (0.25 = 90, 0.5 = 180, 0.75 = 270, 1=360)", group=Detail, parameterName=Detail2Rotation}
  ) {speed=6.2832}
  n45 = Multiply(
   A=@n37
   B=ScalarParameter() {desc="Detail2 Animated Spin Speed (Pivot point isn't very predictable)", group=Detail, parameterName=Detail2Spinner}
  )
  n46 = TextureSampleParameter2D(UVs=Add(A=Multiply(A=@n30, B=@n41), B=Rotator(Coordinate=@n44, Time=@n45) {speed=6.2832}) {desc=Panner}) {group=Detail, parameterName=Detail2, samplerSource=SSM_Wrap_WorldGroupSettings, texture=&T_Black}
  n47 = LinearInterpolate(A=@n29, B=@n46.RGB, Alpha=Saturate(None=ScalarParameter() {defaultValue=0.25, group=Detail, parameterName=Detail2Strength}))
  n48 = StaticSwitchParameter(True=TextureCoordinate() {coordinateIndex=1}, False=TextureCoordinate()) {defaultValue=true, group=Decals, parameterName=*DecalsUseCylinderUV}
  n49 = VectorParameter() {desc="(R,G) = Tile (X,Y)    (B,A) = Pan (X,Y)", group=Decals, parameterName=Decal2Transform, defaultValue=<rgba 1 1 0 0>}
  n50 = ComponentMask(None=@n49.RGB) {g=true, r=true}
  n51 = Time() {bOverride_Period=true, period=20}
  n52 = VectorParameter() {desc="(R,G) = Decal 1 Pan (X,Y)    (B,A) = Decal 2 Pan (X,Y)", group=Decals, parameterName=DecalBases1*2Panner}
  n53 = Add(A=Add(A=Multiply(A=@n48, B=@n50), B=AppendVector(A=@n49.B, B=@n49.B)), B=Multiply(A=@n51, B=AppendVector(A=@n52.B, B=@n52.B))) {desc=Panner}
  n54 = ScalarParameter() {group=Decals, parameterName=Decal2Rotation}
  n55 = Time() {bOverride_Period=true, period=20}
  n56 = Rotator(
   Coordinate=StaticSwitchParameter(True=Rotator(Coordinate=@n53, Time=@n54) {speed=6.2832}, False=@n53) {desc=Enables Decal Sharpness and Threshold params., group=Decals, parameterName=Decal2Rotate}
   Time=Multiply(A=@n55, B=ScalarParameter() {desc="Decal2 Animated Spin Speed (Pivot point isn't very predictable)", group=Decals, parameterName=Decal2Spin})
  ) {speed=6.2832}
  n57 = ScalarParameter() {defaultValue=96, group=Pixelate, parameterName=NumPixels}
  n58 = TextureSampleParameter2D(UVs=StaticSwitchParameter(
    True=MaterialFunctionCall!(UV=@n56, NumPixels=@n57) {materialFunction=/VehicleCosmetics/Materials/ExtraFunctions/PixelateUV_MF}.Result
    False=@n56
   ) {group=Pixelate, parameterName=*PixelateDecal2}) {group=Decals, parameterName=Decal2, texture=&T_Black}
  n59 = StaticComponentMaskParameter(None=StaticSwitchParameter(True=OneMinus(None=@n58.RGB), False=@n58.RGB) {group=Decals, parameterName=Decal2Invert}) {defaultR=true, group=Decals, parameterName=Decal2Channel}
  n60 = Constant() {r=1}
  n61 = Add(
   A=Subtract(A=@n59, B=@n60)
   B=ScalarParameter() {defaultValue=0.5, desc=Tied to HeightMask switch., group=Decals, parameterName=Decal2Threshold}
  )
  n62 = ScalarParameter() {defaultValue=4, desc=Tied to HeightMask switch., group=Decals, parameterName=Decal2Sharpness}
  n63 = StaticSwitchParameter(True=Saturate(None=Multiply(A=@n61, B=@n62)), False=@n59) {desc=Enables Decal Sharpness and Threshold params., group=Decals, parameterName=Decal2HeightMask}
  n64 = TextureSampleParameter2D() {group=BaseTextures, parameterName=BodyTextures, samplerSource=SSM_Wrap_WorldGroupSettings, texture=&T_Black}
  n65 = StaticSwitchParameter(True=TextureCoordinate() {coordinateIndex=1, desc="UV1 (Cylinder)"}, False=TextureCoordinate() {desc="UV0 (Body Unwrap)"}) {defaultValue=true, group=Decals, parameterName=DecalBlendNoise*UseCylinderUV}
  n66 = StaticSwitchParameter(
   True=MaterialFunctionCall(UV=@n65, NumPixels=ScalarParameter() {defaultValue=96, group=Pixelate, parameterName=NumPixels}) {materialFunction=&PixelateUV_MF}.Result
   False=@n65
  ) {group=Pixelate, parameterName=*PixelateDecalBlendNoise}
  n67 = AppendVector(A=ComponentMask(None=@n66) {g=true}, B=ComponentMask(None=@n66) {r=true})
  n68 = ScalarParameter() {group=Decals, parameterName=DecalBlendRampRotation}
  n69 = StaticSwitchParameter(True=LinearInterpolate(A=@n67, B=@n66, Alpha=@n68), False=Rotator(Coordinate=@n67, Time=@n68) {speed=1}) {defaultValue=true, group=Decals, parameterName=DecalBlendRampCheapRotation}
  n70 = ComponentMask(None=@n69) {r=true}
  n71 = Constant() {r=-1}
  n72 = Constant() {r=1}
  n73 = Divide(A=Constant() {r=1}, B=ScalarParameter() {defaultValue=5, group=Decals, parameterName=DecalBlendRampWidth}) {desc=Invert multiplier}
  n74 = Divide(A=@n72, B=@n73) {desc=Get multiplier}
  n75 = ScalarParameter() {defaultValue=0.5, group=Decals, parameterName=DecalBlendRampPosition}
  n76 = Add(A=@n74, B=@n72)
  n77 = Multiply(A=Multiply(A=ConstantBiasScale(None=@n75) {bias=0, scale=2}, B=@n76), B=Constant() {r=0.5})
  n78 = Add(
   A=Add(
    A=StaticSwitchParameter(True=@n64.R, False=@n70) {desc=Instead of a linear ramp., group=Decals, parameterName=DecalBlend*UseEdgeDist}
    B=Subtract(A=@n71, B=@n74)
   )
   B=Add(A=Fmod(A=@n77, B=@n76), B=@n74)
  )
  n79 = Subtract(
   A=StaticSwitchParameter(True=Multiply(A=@n78, B=@n73), False=@n75) {defaultValue=true, desc=Instead of any ramp., group=Decals, parameterName=DecalBlend*UseRamp}
   B=Constant() {r=1}
  )
  n80 = VectorParameter() {desc="(R,G) = Tile (X,Y)    (B,A) = Pan (X,Y)", group=Decals, parameterName=DecalBlendNoiseTransform, defaultValue=<rgba 2 2 0 0>}
  n81 = AppendVector(A=@n80.B, B=@n80.B)
  n82 = StaticSwitchParameter(True=Multiply(A=@n81, B=Time() {bOverride_Period=true, period=20}), False=@n81) {desc=DecalBlendNoiseTransform vector parameter Blue and Alpha control pan speed., group=Decals, parameterName=DecalBlendNoiseDynamic}
  n83 = Add(A=Multiply(A=@n69, B=ComponentMask(None=@n80.RGB) {g=true, r=true}), B=@n82) {desc=Panner}
  n84 = StaticSwitchParameter(True=@n46.RGB, False=@n39.RGB) {group=Decals, parameterName=DecalBlendDistortUseDetail2}
  n85 = Add(
   A=@n83
   B=Multiply(A=ComponentMask(None=@n84) {g=true, r=true}, B=ScalarParameter() {defaultValue=0.25, group=Decals, parameterName=DecalBlendDistortStrength})
  )
  n86 = Rotator(
   Coordinate=Rotator(
    Coordinate=StaticSwitchParameter(True=@n85, False=@n83) {group=Decals, parameterName=DecalBlendDistortByDetail}
    Time=ScalarParameter() {desc="Rotate Decal Blend Noise (0.25 = 90, 0.5 = 180, 0.75 = 270, 1=360)", group=Decals, parameterName=DecalBlendNoiseRotation}
   ) {speed=6.2832}
   Time=Multiply(
    A=Time() {bOverride_Period=true, period=20}
    B=ScalarParameter() {desc="DecalBlendNoise Animated Spin Speed (Pivot point isn't very predictable)", group=Decals, parameterName=DecalBlendNoiseSpin}
   )
  ) {speed=6.2832}
  n87 = StaticComponentMaskParameter(None=TextureSampleParameter2D(UVs=@n86) {group=Decals, parameterName=DecalBlendNoise, texture=&T_Black}.RGB) {defaultR=true, desc=Uses FXNoise2, group=Decals, parameterName=DecalBlendNoiseChannel}
  n88 = Constant()
  n89 = StaticSwitchParameter(True=StaticSwitchParameter(True=OneMinus(None=@n87), False=@n87) {group=Decals, parameterName=DecalBlendNoiseInvert}, False=@n88) {defaultValue=true, group=Decals, parameterName=DecalBlend*UseNoise}
  n90 = ScalarParameter() {defaultValue=32, group=Decals, parameterName=DecalBlendRampNoiseSharpness}
  n91 = Multiply(
   A=Add(A=Add(A=@n79, B=@n89), B=StaticSwitchParameter(True=@n63, False=@n88) {group=Decals, parameterName=DecalBlend*UseDecal2Height})
   B=@n90
  )
  n92 = Saturate(None=@n91)
  n93 = VectorParameter() {desc="(R,G) = Tile (X,Y)    (B,A) = Pan (X,Y)", group=Decals, parameterName=Decal1Transform, defaultValue=<rgba 1 1 0 0>}
  n94 = ComponentMask(None=@n93.RGB) {g=true, r=true}
  n95 = Add(A=Add(A=Multiply(A=@n48, B=@n94), B=AppendVector(A=@n93.B, B=@n93.B)), B=Multiply(A=@n51, B=AppendVector(A=@n52.R, B=@n52.R))) {desc=Panner}
  n96 = ScalarParameter() {group=Decals, parameterName=Decal1Rotation}
  n97 = Rotator(
   Coordinate=StaticSwitchParameter(True=Rotator(Coordinate=@n95, Time=@n96) {speed=6.2832}, False=@n95) {desc=Enables Decal Sharpness and Threshold params., group=Decals, parameterName=Decal1Rotate}
   Time=Multiply(
    A=@n55
    B=ScalarParameter() {desc="Decal 1 Animated Spin Speed (Pivot point isn't very predictable)", group=Decals, parameterName=Decal1Spin}
   )
  ) {speed=6.2832}
  n98 = TextureSampleParameter2D(UVs=StaticSwitchParameter(
    True=MaterialFunctionCall(UV=@n97, NumPixels=@n57) {materialFunction=&PixelateUV_MF}.Result
    False=@n97
   ) {group=Pixelate, parameterName=*PixelateDecal1}) {group=Decals, parameterName=Decal1, texture=&T_Black}
  n99 = StaticComponentMaskParameter(None=StaticSwitchParameter(True=OneMinus(None=@n98.RGB), False=@n98.RGB) {group=Decals, parameterName=Decal1Invert}) {defaultR=true, group=Decals, parameterName=Decal1Channel}
  n100 = Add(
   A=Subtract(A=@n99, B=@n60)
   B=ScalarParameter() {defaultValue=0.5, desc=Tied to HeightMask switch., group=Decals, parameterName=Decal1Threshold}
  )
  n101 = ScalarParameter() {defaultValue=4, desc=Tied to HeightMask switch., group=Decals, parameterName=Decal1Sharpness}
  n102 = StaticSwitchParameter(True=Saturate(None=Multiply(A=@n100, B=@n101)), False=@n99) {desc=Enables Decal Sharpness and Threshold params., group=Decals, parameterName=Decal1HeightMask}
  n103 = StaticSwitchParameter(True=LinearInterpolate(A=@n102, B=@n63, Alpha=@n92), False=@n102) {defaultValue=true, group=Decals, parameterName=*DecalLayer2}
  n104 = Constant()
  n105 = StaticSwitchParameter(True=@n103, False=@n104) {defaultValue=true, group=Decals, parameterName=*DecalLayer1}
  n106 = StaticSwitchParameter(
   True=StaticSwitchParameter(True=LinearInterpolate(A=@n40, B=@n47, Alpha=@n105), False=@n40) {group=Detail, parameterName=*Detail2}
   False=@n29
  ) {defaultValue=true, group=Detail, parameterName=*Detail}
  n107 = ComponentMask(None=@n106) {g=true, r=true}
  n108 = ScalarParameter() {defaultValue=0.125, group=FX, parameterName=FXDistortByDetailNormal}
  n109 = StaticSwitchParameter(True=Add(A=@n28, B=Multiply(A=@n107, B=@n108)), False=@n28) {group=FX, parameterName=FXDistortByDetailNormal}
  n110 = Add(A=@n109, B=LinearInterpolate(B=ScalarParameter() {group=FX, parameterName=FXTwist}, Alpha=ComponentMask(None=@n109) {g=true}))
  n111 = Rotator(
   Coordinate=@n110
   Time=ScalarParameter() {desc="Rotate FX Noise1 (0.25 = 90, 0.5 = 180, 0.75 = 270, 1=360)", group=FX, parameterName=FX*Noise1Rotation}
  ) {speed=6.2832}
  n112 = Time() {bOverride_Period=true, period=20}
  n113 = Multiply(
   A=ScalarParameter() {desc="FX1 Animated Spin Speed  (Pivot point isn't very predictable)", group=FX, parameterName=FX*Noise1Spinner}
   B=@n112
  )
  n114 = VectorParameter() {desc="(R,G) = Tile (X,Y)    (B,A) = Pan (X,Y)", group=FX, parameterName=FX*Noise1Transform, defaultValue=<rgba 4 12 0 0>}
  n115 = AppendVector(A=@n114.B, B=@n114.B)
  n116 = Time() {bOverride_Period=true, period=20}
  n117 = StaticSwitchParameter(True=Multiply(A=@n115, B=@n116), False=@n115) {defaultValue=true, desc=FXNoiseTransform vector parameter Blue and Alpha control pan speed., group=FX, parameterName=FX*Noise*Dynamic}
  n118 = Add(A=Multiply(A=Rotator(Coordinate=@n111, Time=@n113) {speed=6.2832}, B=AppendVector(A=@n114.R, B=@n114.R)), B=@n117) {desc=Panner}
  n119 = ScalarParameter() {defaultValue=96, group=Pixelate, parameterName=NumPixels}
  n120 = StaticSwitchParameter(
   True=MaterialFunctionCall(UV=@n118, NumPixels=@n119) {materialFunction=&PixelateUV_MF}.Result
   False=@n118
  ) {group=Pixelate, parameterName=*PixelateFX}
  n121 = Rotator(
   Coordinate=@n110
   Time=ScalarParameter() {desc="Rotate FX Noise2 (0.25 = 90, 0.5 = 180, 0.75 = 270, 1=360)", group=FX, parameterName=FX*Noise2Rotation}
  ) {speed=6.2832}
  n122 = Multiply(A=@n112, B=ScalarParameter() {desc="FX2 Animated Spin Speed (Pivot point isn't very predictable)", group=FX, parameterName=FX*Noise2Spinner})
  n123 = VectorParameter() {desc="(R,G) = Tile (X,Y)    (B,A) = Pan (X,Y)", group=FX, parameterName=FX*Noise2Transform, defaultValue=<rgba 1 1 0 0>}
  n124 = AppendVector(A=@n123.B, B=@n123.B)
  n125 = StaticSwitchParameter(True=Multiply(A=@n116, B=@n124), False=@n124) {defaultValue=true, desc=FXNoiseTransform vector parameter Blue and Alpha control pan speed., group=FX, parameterName=FX*Noise*Dynamic}
  n126 = Add(A=Multiply(A=Rotator(Coordinate=@n121, Time=@n122) {speed=6.2832}, B=AppendVector(A=@n123.R, B=@n123.R)), B=@n125) {desc=Panner}
  n127 = MaterialFunctionCall(UV=@n126, NumPixels=@n119) {materialFunction=&PixelateUV_MF}
  n128 = Multiply(A=Add(A=@n105, B=@n92), B=ScalarParameter() {defaultValue=0.25, group=FX, parameterName=FXDistortByDecal})
  n129 = TextureSampleParameter2D(UVs=Add(A=StaticSwitchParameter(True=@n127.Result, False=@n126) {group=Pixelate, parameterName=*PixelateFX}, B=@n128)) {group=FX, parameterName=FX*Noise2, samplerSource=SSM_Wrap_WorldGroupSettings, texture=&T_Black}
  n130 = StaticComponentMaskParameter(None=StaticSwitchParameter(True=OneMinus(None=@n129.RGB), False=@n129.RGB) {group=FX, parameterName=FX*Noise2Invert}) {defaultG=true, group=FX, parameterName=FX*Noise2Channel}
  n131 = Add(A=@n120, B=Multiply(A=@n130, B=ScalarParameter() {defaultValue=0.25, group=FX, parameterName=FXDistortBySelf}))
  n132 = TextureSampleParameter2D(UVs=Add(A=StaticSwitchParameter(True=@n131, False=@n120) {defaultValue=true, group=FX, parameterName=*FXLayer2}, B=@n128)) {group=FX, parameterName=FX*Noise1, samplerSource=SSM_Wrap_WorldGroupSettings, texture=&T_Black}
  n133 = StaticComponentMaskParameter(None=StaticSwitchParameter(True=OneMinus(None=@n132.RGB), False=@n132.RGB) {group=FX, parameterName=FX*Noise1Invert}) {defaultG=true, group=FX, parameterName=FX*Noise1Channel}
  n134 = LinearInterpolate(A=@n133, B=Constant() {r=1}, Alpha=@n130)
  n135 = Add(A=@n133, B=@n130)
  n136 = AppendVector(A=AppendVector(A=AppendVector(A=@n134, B=@n135), B=Subtract(A=@n133, B=@n130)), B=Multiply(A=@n133, B=@n130))
  n137 = Add(
   A=Multiply(A=Subtract(A=@n18, B=@n19), B=Divide(B=@n20) {desc=Invert multiplier})
   B=StaticSwitchParameter(
    True=StaticComponentMaskParameter(None=@n136) {defaultR=true, desc="Add, Subtract, Multiply, Lerp", group=FX, parameterName=FX*LayerMaskBlendType}
    False=@n133
   ) {defaultValue=true, group=FX, parameterName=*FXLayer2}
  )
  n138 = ScalarParameter() {defaultValue=16, group=FX, parameterName=FXFadeSharpness}
  n139 = Saturate(None=Multiply(A=@n138, B=Constant() {r=0.25}))
  n140 = LinearInterpolate(
   A=@n11
   B=@n12
   Alpha=Clamp(None=Multiply(A=@n137, B=@n139), Max=ScalarParameter() {defaultValue=4, group=FX, parameterName=FX*LayerOverBlend})
  )
  n141 = Multiply(A=@n137, B=@n138)
  n142 = Saturate(None=Add(A=@n141, B=@n141))
  n143 = LinearInterpolate(A=@n1, B=@n140, Alpha=@n142)
  n144 = VectorParameter() {desc="Alpha = Emissive Color Brightness", group=Colors, parameterName=Color2, defaultValue=<rgba 1 0.8147 0.1228 1>}
  n145 = StaticSwitchParameter(True=StaticSwitchParameter(True=@n2.RGB, False=@n4) {group=Decals, parameterName=UsePaintColor_Decal_2}, False=@n144.RGB) {group=Colors, parameterName=Use2Colors}
  n146 = VectorParameter() {desc="Alpha = Emissive Color Brightness", group=Colors, parameterName=Color1, defaultValue=<rgba 0.0316 0.0651 0.0829 1>}
  n147 = StaticSwitchParameter(True=StaticSwitchParameter(True=@n2.RGB, False=@n8) {group=Decals, parameterName=UsePaintColor_Decal_1}, False=@n146.RGB) {group=Colors, parameterName=Use2Colors}
  n148 = StaticSwitchParameter(True=@n145, False=@n147) {group=Colors, parameterName=SwapDecalColors}
  n149 = StaticSwitchParameter(True=@n147, False=@n145) {group=Colors, parameterName=SwapDecalColors}
  n150 = LinearInterpolate(A=@n148, B=@n149, Alpha=StaticSwitchParameter(True=@n102, False=@n104) {defaultValue=true, group=Decals, parameterName=*DecalLayer1})
  n151 = Multiply(A=@n150, B=LinearInterpolate(A=Constant() {r=1}, B=@n150, Alpha=@n92))
  n152 = StaticSwitchParameter(True=@n149, False=@n148) {desc=Mutually Exclusive, group=Outlines, parameterName=OutlineUseColor2}
  n153 = VectorParameter() {desc="Alpha = Emissive Color Brightness", group=Colors, parameterName=Color4, defaultValue=<rgba 1 0.1885 0.5394 16>}
  n154 = StaticSwitchParameter(True=@n8, False=@n153.RGB) {group=Colors, parameterName=Use2Colors}
  n155 = VectorParameter() {desc="Alpha = Emissive Color Brightness", group=Colors, parameterName=Color3, defaultValue=<rgba 0.0154 0.0385 0.105 1>}
  n156 = StaticSwitchParameter(True=@n4, False=@n155.RGB) {group=Colors, parameterName=Use2Colors}
  n157 = StaticSwitchParameter(
   True=@n11
   False=StaticSwitchParameter(True=@n154, False=StaticSwitchParameter(True=@n156, False=@n152) {desc=Mutually Exclusive, group=Outlines, parameterName=OutlineUseColor3}) {desc=Mutually Exclusive, group=Outlines, parameterName=OutlineUseColor4}
  ) {desc=Mutually Exclusive, group=Outlines, parameterName=OutlineUseFXColor1}
  n158 = StaticSwitchParameter(
   True=@n152
   False=StaticSwitchParameter(True=@n140, False=StaticSwitchParameter(True=@n12, False=@n157) {desc=Mutually Exclusive, group=Outlines, parameterName=OutlineUseFXColor2}) {desc=Mutually Exclusive, group=Outlines, parameterName=OutlineUseBothFXColors}
  ) {group=Colors, parameterName=Use2Colors}
  n159 = Constant()
  n160 = Constant()
  n161 = TextureSampleParameter2D(UVs=BumpOffset(Coordinate=@n86, Height=@n160, HeightRatioInput=ScalarParameter() {defaultValue=0.0156, group=Outlines, parameterName=DropShadowDistance}) {heightRatio=0.25}) {group=Decals, parameterName=DecalBlendNoise, samplerSource=SSM_Wrap_WorldGroupSettings, texture=&T_Black}
  n162 = StaticComponentMaskParameter(None=@n161.RGB) {defaultR=true, desc=Uses FXNoise2, group=Decals, parameterName=DecalBlendNoiseChannel}
  n163 = OneMinus(None=@n162)
  n164 = Add(
   A=@n79
   B=StaticSwitchParameter(True=StaticSwitchParameter(True=@n163, False=@n162) {group=Decals, parameterName=DecalBlendNoiseInvert}, False=@n160) {defaultValue=true, group=Decals, parameterName=DecalBlend*UseNoise}
  )
  n165 = OneMinus(None=Saturate(None=Multiply(A=@n164, B=@n90)))
  n166 = Constant() {r=1}
  n167 = LinearInterpolate(
   A=@n150
   B=StaticSwitchParameter(True=@n158, False=@n159) {group=Outlines, parameterName=DropShadowUseOutlineColor}
   Alpha=OneMinus(None=StaticSwitchParameter(True=@n165, False=@n166) {desc=Requires use of Skin2, group=Outlines, parameterName=*DropShadow})
  )
  n168 = LinearInterpolate(A=@n156, B=@n154, Alpha=StaticSwitchParameter(True=@n103, False=@n104) {defaultValue=true, group=Decals, parameterName=*DecalLayer1})
  n169 = Multiply(A=@n168, B=@n92)
  n170 = Transform(None=CameraVectorWS()) {transformSourceType=TRANSFORMSOURCE_World, transformType=TRANSFORM_Tangent}
  n171 = Constant3Vector() {constant=<rgba 1 1 0 0>}
  n172 = Constant() {desc=Original BodyN B, r=1}
  n173 = StaticSwitchParameter(True=Add(A=Multiply(A=@n106, B=@n171), B=AppendVector(A=@n26, B=@n172) {desc=R+G+B}), False=@n106) {group=Lighting, parameterName=*LightUseBodyNormal}
  n174 = DotProduct(A=@n170, B=@n173)
  n175 = OneMinus(None=@n174)
  n176 = StaticSwitchParameter(True=LinearInterpolate(A=Multiply(A=@n175, B=@n175), B=@n174, Alpha=@n92), False=@n92) {defaultValue=true, group=Colors, parameterName=Pearlescent}
  n177 = StaticSwitchParameter(
   True=Subtract(A=@n167, B=@n169)
   False=StaticSwitchParameter(True=Add(A=@n167, B=@n169), False=LinearInterpolate(A=@n167, B=@n168, Alpha=@n176)) {group=Decals, parameterName=DecalBlend*Add}
  ) {group=Decals, parameterName=DecalBlend*Subtract}
  n178 = StaticSwitchParameter(True=Saturate(None=StaticSwitchParameter(True=@n151, False=@n177) {group=Decals, parameterName=DecalBlend*Multiply}), False=@n150) {defaultValue=true, group=Decals, parameterName=*DecalLayer2}
  n179 = Multiply(A=@n178, B=Saturate(None=StaticSwitchParameter(True=@n143, False=@n1) {defaultValue=true, group=FX, parameterName=*FXLayer1}))
  n180 = StaticSwitchParameter(True=Multiply(A=@n140, B=@n142), False=Constant()) {defaultValue=true, group=FX, parameterName=*FXLayer1}
  n181 = StaticSwitchParameter(
   True=Subtract(A=@n178, B=@n180)
   False=StaticSwitchParameter(True=Add(A=@n178, B=@n180), False=LinearInterpolate(A=@n178, B=@n140, Alpha=@n142)) {group=FX, parameterName=FX*BlendAdd}
  ) {group=FX, parameterName=FX*BlendSubtract}
  n182 = StaticSwitchParameter(True=Saturate(None=StaticSwitchParameter(True=@n179, False=@n181) {group=FX, parameterName=FX*BlendMultiply}), False=@n178) {defaultValue=true, group=FX, parameterName=*FXLayer1}
  n183 = OneMinus(None=DotProduct(A=@n170, B=Constant3Vector() {constant=<rgba 0 0 1 0>}))
  n184 = Constant() {r=-1}
  n185 = Multiply(
   A=Add(A=Add(A=@n183, B=@n184), B=ScalarParameter() {defaultValue=0.5, group=Outlines, parameterName=Outline*SilhouetteWidth})
   B=ScalarParameter() {defaultValue=32, group=Outlines, parameterName=Outline*SilhouetteSharpness}
  )
  n186 = Constant()
  n187 = AppendVector(A=@n13.B, B=@n64.R)
  n188 = StaticComponentMaskParameter(None=AppendVector(A=AppendVector(A=@n187, B=@n64.G), B=@n64.G)) {defaultR=true, group=Outlines, parameterName=OutlineEdgeMaskChannel}
  n189 = StaticSwitchParameter(True=OneMinus(None=@n188), False=@n188) {group=Outlines, parameterName=OutlineEdgeMaskInvert}
  n190 = Constant() {r=-1}
  n191 = ScalarParameter() {group=Outlines, parameterName=Outline*Width}
  n192 = Constant() {r=0.5}
  n193 = Multiply(
   A=Add(A=Add(A=@n189, B=@n190), B=Multiply(A=@n191, B=@n192))
   B=ScalarParameter() {defaultValue=8, group=Outlines, parameterName=Outline*Sharpness}
  )
  n194 = Add(
   A=StaticSwitchParameter(True=Saturate(None=@n185), False=@n186) {group=Outlines, parameterName=OutlineSilhouette}
   B=StaticSwitchParameter(True=Saturate(None=@n193), False=@n186) {group=Outlines, parameterName=OutlineEdges}
  )
  n195 = Abs(None=ConstantBiasScale(None=@n105) {bias=-0.5, scale=4})
  n196 = StaticSwitchParameter(True=Saturate(None=OneMinus(None=@n195)), False=@n186) {desc=Doesn't work well with sharpened height mask decals., group=Outlines, parameterName=OutlineDecals}
  n197 = Saturate(None=Multiply(A=Add(A=@n194, B=@n196), B=Constant() {r=8}))
  n198 = Saturate(None=ConstantBiasScale(None=Abs(None=@n91)) {bias=-0.5, scale=4})
  n199 = Constant() {r=1}
  n200 = StaticSwitchParameter(
   True=Multiply(
    A=OneMinus(None=@n197)
    B=StaticSwitchParameter(True=@n198, False=@n199) {defaultValue=true, group=Decals, parameterName=DecalBlend*UseHeightBlend}
   )
   False=Constant() {r=1}
  ) {group=Outlines, parameterName=*Outlines}
  n201 = TextureCoordinate()
  n202 = ScalarParameter() {defaultValue=96, group=Pixelate, parameterName=NumPixels}
  n203 = ConstantBiasScale(None=Fmod(A=Multiply(A=@n201, B=@n202), B=Constant() {r=1})) {bias=-0.5, scale=2}
  n204 = Saturate(None=OneMinus(None=Abs(None=@n203)))
  n205 = Multiply(A=ComponentMask(None=@n204) {r=true}, B=ComponentMask(None=@n204) {g=true})
  n206 = StaticSwitchParameter(True=Saturate(None=ConstantBiasScale(None=@n205) {bias=0, scale=4}), False=Constant() {r=1}) {group=Outlines, parameterName=OutlinePixels}
  n207 = OneMinus(None=Multiply(A=OneMinus(None=@n39.A), B=ScalarParameter() {group=Detail, parameterName=Detail1Strength_AO}))
  n208 = OneMinus(None=Multiply(A=ScalarParameter() {group=Detail, parameterName=Detail2Strength_AO}, B=OneMinus(None=@n46.A)))
  n209 = Multiply(
   A=Multiply(A=Multiply(A=@n182, B=@n200) {desc=Outlines}, B=@n206) {desc=Pixel Outlines}
   B=Multiply(A=@n64.G, B=LinearInterpolate(A=@n207, B=@n208, Alpha=@n105)) {desc=Combine AO}
  ) {desc=AO}
  n210 = Saturate(None=@n209)
  n211 = ScalarParameter() {group=Metallic, parameterName=Metallic, sliderMax=1, sliderMin=0.01}
  n212 = ScalarParameter() {group=Metallic, parameterName=MetallicMin}
  n213 = ScalarParameter() {defaultValue=1, group=Metallic, parameterName=MetallicMax}
  n214 = ScalarParameter() {defaultValue=0.35, desc=Body, group=Roughness, parameterName=Roughness}
  n215 = ScalarParameter() {defaultValue=0.01, group=Roughness, parameterName=RoughMin}
  n216 = ScalarParameter() {defaultValue=0.85, group=Roughness, parameterName=RoughMax}
  n217 = LinearInterpolate(A=@n4, B=@n8, Alpha=ScalarParameter() {parameterName=EmissiveUseTeamColor})
  n218 = ScalarParameter() {parameterName=EmissiveUsePaintColor}
  n219 = Multiply(A=Multiply(A=LinearInterpolate(A=@n217, B=@n2.RGB, Alpha=@n218), B=ScalarParameter() {parameterName=BodyEmissiveBrightness}), B=@n64.G)
  n220 = Multiply(A=LinearInterpolate(A=StaticSwitchParameter(True=@n158, False=@n159) {group=Outlines, parameterName=OutlineUseColor}, B=@n180, Alpha=@n200))
  n221 = Multiply(A=Saturate(None=Multiply(A=@n220, B=@n206)), B=ScalarParameter() {defaultValue=1, group=Colors, parameterName=FXColorEmissive})
  MaterialFunctionCall!(
   Fresnel
   Material=MakeMaterialAttributes!(
    BaseColor=StaticSwitchParameter(True=@n210, False=@n209) {group=Colors, parameterName=ColorClamp}
    Metallic=Clamp(None=@n211, Min=@n212, Max=@n213)
    Specular
    Roughness=Clamp(None=@n214, Min=@n215, Max=@n216)
    Anisotropy
    EmissiveColor=Add(A=@n219, B=@n221)
    Opacity
    OpacityMask
    Normal=@n173
    Tangent
    WorldPositionOffset
    SubsurfaceColor
    ClearCoat
    ClearCoatRoughness
    AmbientOcclusion
    Refraction
    CustomizedUVs_0
    CustomizedUVs_1
    CustomizedUVs_2
    CustomizedUVs_3
    CustomizedUVs_4
    CustomizedUVs_5
    CustomizedUVs_6
    CustomizedUVs_7
    PixelDepthOffset
    ShadingModel
    Displacement
   )
  ) {materialFunction=/Game/Packages/Fortress_SharedMaterials/Base_Material_Functions/CharacterDeRez_Transition}
  VectorParameter() {desc="Alpha = Emissive Color Brightness", group=Colors, parameterName=CustomColor, defaultValue=<rgba 1 0.8147 0.1228 1>}
 ]
}
